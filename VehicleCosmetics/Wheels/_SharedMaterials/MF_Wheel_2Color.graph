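# Material function for vehicle cosmetic wheels that paints rim and tread with custom and accent colors, returning full material attributes.
# RGB textures pack paint masks, AO and curvature, with optional floored or flipped masks.
# Static switches choose one or two colors and texture or scalar PBR values, plus reflection-vector sparkles.
# Rim and tread results are lerped into base color, metallic, specular, roughness, normal and emissive, apparently by vertex red.
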MaterialFunction /VehicleCosmetics/Wheels/_SharedMaterials/MF_Wheel_2Color
graph {nodes [
  n1 = TextureSampleParameter2D() {group=Textures, parameterName=RimRGB, samplerType=SAMPLERTYPE_LinearColor, texture=/VehicleCosmetics/Textures/T_Black_Linear}
  n2 = ScalarParameter() {group=Colors, parameterName=Rim_PaintedAccentFloorValue}
  Composite!(InputChannel=@n1.G, FloorValue=@n2) {subgraphName=FloorAccentColor}
  Composite(InputChannel=@n1.G, FloorValue=@n2) {subgraphName=FloorAccentColorFlip}
  n3 = ScalarParameter() {group=Colors, parameterName=Rim_PaintedCustomFloorValue}
  Composite(InputChannel=@n1.R, FloorValue=@n3) {subgraphName=FloorCustomColorFlip}
  n4 = TextureSampleParameter2D(UVs=NamedRerouteUsage()) {group=Textures, parameterName=TreadRGB, samplerType=SAMPLERTYPE_LinearColor, texture=&T_Black_Linear}
  n5 = ScalarParameter() {group=Colors, parameterName=Tread_PaintedAccentFloorValue}
  Composite(InputChannel=@n4.G, FloorValue=@n5) {subgraphName=TreadFloorAccentColorFlip}
  Composite(InputChannel=@n1.R, FloorValue=@n3) {subgraphName=FloorCustomColor}
  Composite(InputChannel=@n4.G, FloorValue=@n5) {subgraphName=TreadFloorAccentColor}
  n6 = ScalarParameter() {group=Colors, parameterName=Tread_PaintedCustomFloorValue}
  Composite(InputChannel=@n4.R, FloorValue=@n6) {subgraphName=TreadFloorCustomColorFlip}
  Composite(InputChannel=@n4.R, FloorValue=@n6) {subgraphName=TreadFloorCustomColor}
  MaterialExpressionReroute_21 = Reroute(None=@n1.G)
  MaterialExpressionReroute_22 = Reroute(None=@n2)
  n7 = Add(A=@MaterialExpressionReroute_22.FloorValue) {constB=1.5}
  Floor(None=Saturate(None=Multiply(A=@MaterialExpressionReroute_21.InputChannel, B=@n7) {constB=1.5}))
  MaterialExpressionReroute_11 = Reroute(None=@n4.G)
  MaterialExpressionReroute_12 = Reroute(None=@n5)
  n8 = Add(A=@MaterialExpressionReroute_12.FloorValue) {constB=1.5}
  Floor(None=Saturate(None=Multiply(A=@MaterialExpressionReroute_11.InputChannel, B=@n8) {constB=1.5}))
  MaterialExpressionReroute_17 = Reroute(None=@n4.R)
  MaterialExpressionReroute_18 = Reroute(None=@n6)
  n9 = Add(A=@MaterialExpressionReroute_18.FloorValue) {constB=1.5}
  Floor(None=Saturate(None=Multiply(A=@MaterialExpressionReroute_17.InputChannel, B=@n9) {constB=1.5}))
  n10 = Constant3Vector()
  n11 = Add(A=Constant()) {constB=1.5}
  Floor(None=Saturate(None=Multiply(A=@n10, B=@n11) {constB=1.5}))
  FunctionOutput(None=MakeMaterialAttributes!(
    BaseColor=NamedRerouteUsage()
    Metallic=NamedRerouteUsage()
    Specular=NamedRerouteUsage()
    Roughness=NamedRerouteUsage()
    Anisotropy
    EmissiveColor=NamedRerouteUsage()
    Opacity
    OpacityMask
    Normal=NamedRerouteUsage()
    Tangent
    WorldPositionOffset
    SubsurfaceColor
    ClearCoat
    ClearCoatRoughness
    AmbientOcclusion
    Refraction
    CustomizedUVs_0
    CustomizedUVs_1
    CustomizedUVs_2
    CustomizedUVs_3
    CustomizedUVs_4
    CustomizedUVs_5
    CustomizedUVs_6
    CustomizedUVs_7
    PixelDepthOffset
    ShadingModel
    Displacement
   ))
  n12 = NamedRerouteUsage()
  n13 = StaticSwitchParameter(True=@n12, False=Constant() {r=1}) {parameterName=PaintMaskSpec}
  n14 = Multiply(A=ScalarParameter() {defaultValue=10, group=Rim_Properties, parameterName=RimSpecBrightness}, B=NamedRerouteUsage())
  Multiply(A=@n12, B=Saturate(None=Multiply(A=@n13, B=@n14)))
  n15 = ComponentMask() {b=true, desc=Prime ENV Input, g=true, r=true}
  n16 = ScalarParameter() {defaultValue=0.36, group=Tread_Properties, parameterName=TreadReflectionBrightness}
  n17 = NamedRerouteUsage()
  n18 = ScalarParameter() {defaultValue=1, group=TA_ONLY, parameterName=TreadAccentColor_Brightness}
  n19 = NamedRerouteUsage()
  n20 = NamedRerouteUsage()
  n21 = ScalarParameter() {defaultValue=1, group=TA_ONLY, parameterName=TreadColor_Brightness}
  n22 = StaticSwitchParameter(True=Multiply(A=ComponentMask(None=@n19) {b=true}, B=Multiply(A=@n20, B=@n21)), False=NamedRerouteUsage()) {group=Colors, parameterName=UseTreadColor}
  n23 = NamedRerouteUsage()
  n24 = LinearInterpolate(
   A=@n22
   B=Multiply(A=NamedRerouteUsage(), B=ScalarParameter() {defaultValue=1, group=TA_ONLY, parameterName=TreadCustomColor_Brightness})
   Alpha=ComponentMask(None=@n23) {r=true}
  ) {desc=L3}
  n25 = StaticSwitchParameter(
   True=LinearInterpolate(A=@n24, B=Multiply(A=@n17, B=@n18), Alpha=ComponentMask(None=@n23) {g=true})
   False=StaticSwitchParameter(True=@n24, False=@n22) {group=Colors, parameterName=TreadPainted}
  ) {group=Colors, parameterName=TreadPaintUses2Colors}
  n26 = StaticSwitchParameter(True=Multiply(A=Add(A=@n15, B=@n16), B=@n25), False=@n25) {group=Tread_Properties, parameterName=TreadReflection}
  Multiply(A=@n26, B=@n25)
  NamedRerouteDeclaration(None=LinearInterpolate(A=NamedRerouteUsage(), B=NamedRerouteUsage(), Alpha=NamedRerouteUsage())) {name=BC_Output, nodeColor=<rgba 0 0.7656 1 1>}
  n27 = ComponentMask(None=NamedRerouteUsage()) {r=true}
  n28 = LinearInterpolate(
   A=ScalarParameter() {group=Rim_Properties, parameterName=RimMetallic, sortPriority=1}
   B=ScalarParameter() {defaultValue=1, group=Rim_Properties, parameterName=RimMetallic_Painted, sortPriority=1}
   Alpha=NamedRerouteUsage()
  )
  n29 = ComponentMask(None=NamedRerouteUsage()) {r=true}
  n30 = LinearInterpolate(
   A=ScalarParameter() {group=Tread_Properties, parameterName=TreadMetallic, sortPriority=1}
   B=ScalarParameter() {group=Tread_Properties, parameterName=TreadMetallic_Painted, sortPriority=1}
   Alpha=NamedRerouteUsage()
  )
  NamedRerouteDeclaration(None=LinearInterpolate(
    A=StaticSwitchParameter(True=@n27, False=@n28) {parameterName=RimMetallicFromTexture}
    B=StaticSwitchParameter(True=@n29, False=@n30) {parameterName=TreadMetallicFromTexture}
    Alpha=NamedRerouteUsage()
   )) {name=Metallic_Output, nodeColor=<rgba 1 0 0.1875 1>}
  n31 = TextureSampleParameter2D() {
   desc="R=Metallic, G=Rough, B=Spec, A=Unused"
   group=Textures
   parameterName=RimPBR
   samplerType=SAMPLERTYPE_LinearColor
   texture=&T_Black_Linear
  }
  NamedRerouteDeclaration(None=@n31.RGBA) {name=NR_RimPBR, nodeColor=<rgba 0 1 0.1562 1>}
  n32 = Add(A=ReflectionVectorWS())
  n33 = TextureSample(UVs=Multiply(A=ComponentMask(None=@n32) {b=true, g=true, r=true}) {constB=0.0625}) {texture=/DealershipSharedLibrary/FX/Textures/Noise/T_VCVFX_NoiseGeneric01Pack_RGB}
  n34 = TextureSample(UVs=Constant() {r=4}) {texture=&T_VCVFX_NoiseGeneric01Pack_RGB}
  NamedRerouteDeclaration(None=Multiply(A=Multiply(A=@n33.RGB, B=@n34.RGB)) {constB=4096}) {name=NR_Sparkle, nodeColor=<rgba 1 0 0.7969 1>}
  MaterialExpressionReroute_43 = Reroute(None=@n1.R)
  n35 = OneMinus(None=@MaterialExpressionReroute_43.InputChannel)
  n36 = Saturate(None=Multiply(A=AppendVector(A=@n35, B=@MaterialExpressionReroute_43.InputChannel)) {constB=3})
  n37 = ComponentMask(None=@n36) {r=true}
  n38 = ComponentMask(None=@n36) {g=true}
  MaterialExpressionReroute_44 = Reroute(None=@n3)
  n39 = OneMinus(None=Saturate(None=LinearInterpolate(A=@n37, B=@n38, Alpha=@MaterialExpressionReroute_44.FloorValue)))
  MaterialExpressionReroute_45 = Reroute(None=@n39)
  MaterialExpressionReroute_24 = Reroute(None=@n1.R)
  n40 = Add(A=@MaterialExpressionReroute_24.InputChannel) {constB=-0.5}
  n41 = AppendVector(
   A=Multiply(A=Saturate(None=@n40)) {constB=2}
   B=Add(A=@MaterialExpressionReroute_24.InputChannel, B=@MaterialExpressionReroute_24.InputChannel)
  )
  n42 = MaterialFunctionCall(Float2=Saturate(None=Multiply(A=@n41, B=@n41))) {materialFunction=&BreakOutFloat2Components}
  MaterialExpressionReroute_25 = Reroute(None=@n3)
  n43 = LinearInterpolate(A=@n42.R, B=@n42.R, Alpha=@MaterialExpressionReroute_25.FloorValue)
  MaterialExpressionReroute_23 = Reroute(None=@n43)
  n44 = StaticSwitchParameter(True=@MaterialExpressionReroute_45.Output, False=@MaterialExpressionReroute_23.Output) {group=Rim_Properties, parameterName=Rim_FlooredPaintFlip}
  MaterialExpressionReroute_40 = Reroute(None=@n1.G)
  n45 = Add(A=@MaterialExpressionReroute_40.InputChannel) {constB=-0.5}
  n46 = AppendVector(
   A=Multiply(A=Saturate(None=@n45)) {constB=2}
   B=Add(A=@MaterialExpressionReroute_40.InputChannel, B=@MaterialExpressionReroute_40.InputChannel)
  )
  n47 = MaterialFunctionCall(Float2=Saturate(None=Multiply(A=@n46, B=@n46))) {materialFunction=&BreakOutFloat2Components}
  MaterialExpressionReroute_41 = Reroute(None=@n2)
  n48 = LinearInterpolate(A=@n47.R, B=@n47.R, Alpha=@MaterialExpressionReroute_41.FloorValue)
  MaterialExpressionReroute_42 = Reroute(None=@n48)
  n49 = Add(A=@MaterialExpressionReroute_21.InputChannel) {constB=-0.5}
  n50 = AppendVector(
   A=Multiply(A=Saturate(None=@n49)) {constB=2}
   B=Add(A=@MaterialExpressionReroute_21.InputChannel, B=@MaterialExpressionReroute_21.InputChannel)
  )
  n51 = MaterialFunctionCall(Float2=Saturate(None=Multiply(A=@n50, B=@n50))) {materialFunction=&BreakOutFloat2Components}
  n52 = LinearInterpolate(A=@n51.R, B=@n51.R, Alpha=@MaterialExpressionReroute_22.FloorValue)
  MaterialExpressionReroute_20 = Reroute(None=@n52)
  n53 = StaticSwitchParameter(True=@MaterialExpressionReroute_42.Output, False=@MaterialExpressionReroute_20.Output) {group=Rim_Properties, parameterName=Rim_FlooredPaintFlip}
  n54 = StaticSwitchParameter(
   True=MaterialFunctionCall(X=@n44, Y=@n53, Z=@n1.G) {materialFunction=&MakeFloat3}.Result
   False=@n1.RGB
  ) {group=Rim_Properties, parameterName=Rim_FlooredPaintColor}
  NamedRerouteDeclaration(None=OneMinus(None=NamedRerouteDeclaration(None=@n54) {name=NR_RimRGB, nodeColor=<rgba 1 0 0.6562 1>})) {name=NR_RimRGB_Inverted, nodeColor=<rgba 1 0.6562 0 1>}
  NamedRerouteDeclaration() {name=RimSpec_Out, nodeColor=<rgba 0 1 0.2266 1>}
  n55 = Multiply(A=ScalarParameter() {defaultValue=1, group=TA_ONLY, parameterName=RimCustomColor_Brightness}, B=NamedRerouteUsage())
  n56 = Multiply(A=NamedRerouteUsage(), B=ScalarParameter() {defaultValue=1, group=TA_ONLY, parameterName=RimAccentColor_Brightness})
  n57 = ComponentMask(None=NamedRerouteUsage()) {g=true}
  n58 = LinearInterpolate(
   A=NamedRerouteUsage()
   B=StaticSwitchParameter(True=LinearInterpolate(A=@n55, B=@n56, Alpha=@n57), False=@n55) {desc=L1, group=Colors, parameterName=RimPaintUses2Colors}
   Alpha=NamedRerouteUsage()
  )
  n59 = LinearInterpolate(A=@n58, B=Multiply(A=@n58), Alpha=ScalarParameter() {group=Rim_Properties, parameterName=RimPaintReflectionAmount})
  n60 = Multiply(A=NamedRerouteUsage(), B=@n59)
  n61 = ComponentMask(None=NamedRerouteUsage()) {a=true}
  n62 = StaticSwitchParameter(True=Add(A=Multiply(A=@n60, B=@n61), B=@n59), False=@n59) {group=Misc, parameterName=Sparkles}
  n63 = NamedRerouteUsage()
  n64 = Constant() {r=0.5}
  n65 = ComponentMask(None=NamedRerouteUsage()) {b=true, g=true, r=true}
  n66 = NamedRerouteUsage()
  NamedRerouteDeclaration(None=MaterialFunctionCall(Base=Multiply(A=@n62, B=@n63), Blend=LinearInterpolate(A=@n64, B=@n65, Alpha=@n66)) {materialFunction=&Blend_Overlay}.Result) {name=RimDiffuse_Out, nodeColor=<rgba 0 1 0.6719 1>}
  n67 = LinearInterpolate(Alpha=ScalarParameter() {defaultValue=1, parameterName=TreadUVDirection_X}) {constA=-1}
  n68 = LinearInterpolate(Alpha=ScalarParameter() {defaultValue=1, parameterName=TreadUVDirection_Y}) {constA=-1}
  NamedRerouteDeclaration(None=Multiply(A=TextureCoordinate(), B=AppendVector(A=@n67, B=@n68))) {name=TreadUVs, nodeColor=<rgba 0 1 0.4375 1>}
  n69 = NamedRerouteUsage()
  NamedRerouteDeclaration(None=StaticSwitchParameter(True=NamedRerouteUsage(), False=ComponentMask(None=@n69) {r=true}) {desc=L1, group=Colors, parameterName=TreadPaintUses2Colors}) {name=TreadColorizationMasks, nodeColor=<rgba 0 0.5781 1 1>}
  MaterialExpressionReroute_26 = Reroute(None=@n4.R)
  n70 = OneMinus(None=@MaterialExpressionReroute_26.InputChannel)
  n71 = Saturate(None=Multiply(A=AppendVector(A=@n70, B=@MaterialExpressionReroute_26.InputChannel)) {constB=3})
  n72 = ComponentMask(None=@n71) {r=true}
  n73 = ComponentMask(None=@n71) {g=true}
  MaterialExpressionReroute_28 = Reroute(None=@n6)
  n74 = OneMinus(None=Saturate(None=LinearInterpolate(A=@n72, B=@n73, Alpha=@MaterialExpressionReroute_28.FloorValue)))
  MaterialExpressionReroute_29 = Reroute(None=@n74)
  n75 = Add(A=@MaterialExpressionReroute_17.InputChannel) {constB=-0.5}
  n76 = AppendVector(
   A=Multiply(A=Saturate(None=@n75)) {constB=2}
   B=Add(A=@MaterialExpressionReroute_17.InputChannel, B=@MaterialExpressionReroute_17.InputChannel)
  )
  n77 = MaterialFunctionCall(Float2=Saturate(None=Multiply(A=@n76, B=@n76))) {materialFunction=&BreakOutFloat2Components}
  n78 = LinearInterpolate(A=@n77.R, B=@n77.R, Alpha=@MaterialExpressionReroute_18.FloorValue)
  MaterialExpressionReroute_19 = Reroute(None=@n78)
  n79 = StaticSwitchParameter(True=@MaterialExpressionReroute_29.Output, False=@MaterialExpressionReroute_19.Output) {group=Tread_Properties, parameterName=Tread_FlooredPaintFlip}
  MaterialExpressionReroute_37 = Reroute(None=@n4.G)
  n80 = OneMinus(None=@MaterialExpressionReroute_37.InputChannel)
  n81 = Saturate(None=Multiply(A=AppendVector(A=@n80, B=@MaterialExpressionReroute_37.InputChannel)) {constB=3})
  n82 = ComponentMask(None=@n81) {r=true}
  n83 = ComponentMask(None=@n81) {g=true}
  MaterialExpressionReroute_39 = Reroute(None=@n5)
  n84 = OneMinus(None=Saturate(None=LinearInterpolate(A=@n82, B=@n83, Alpha=@MaterialExpressionReroute_39.FloorValue)))
  MaterialExpressionReroute_46 = Reroute(None=@n84)
  n85 = Add(A=@MaterialExpressionReroute_11.InputChannel) {constB=-0.5}
  n86 = AppendVector(
   A=Multiply(A=Saturate(None=@n85)) {constB=2}
   B=Add(A=@MaterialExpressionReroute_11.InputChannel, B=@MaterialExpressionReroute_11.InputChannel)
  )
  n87 = MaterialFunctionCall(Float2=Saturate(None=Multiply(A=@n86, B=@n86))) {materialFunction=&BreakOutFloat2Components}
  n88 = LinearInterpolate(A=@n87.R, B=@n87.R, Alpha=@MaterialExpressionReroute_12.FloorValue)
  MaterialExpressionReroute_13 = Reroute(None=@n88)
  n89 = StaticSwitchParameter(True=@MaterialExpressionReroute_46.Output, False=@MaterialExpressionReroute_13.Output) {group=Tread_Properties, parameterName=Tread_FlooredPaintFlip}
  NamedRerouteDeclaration(None=StaticSwitchParameter(
    True=MaterialFunctionCall(X=@n79, Y=@n89, Z=@n4.G) {materialFunction=&MakeFloat3}.Result
    False=@n4.RGB
   ) {group=Tread_Properties, parameterName=Tread_FlooredPaintColor}) {name=NR_TreadRGB, nodeColor=<rgba 1 0 0.2812 1>}
  n90 = NamedRerouteUsage()
  NamedRerouteDeclaration(None=TextureSampleParameter2D(UVs=@n90) {group=Textures, parameterName=TreadBC, texture=/VehicleCosmetics/Textures/T_Black}.RGB) {name=NR_TreadBC, nodeColor=<rgba 1 0 0.6797 1>}
  n91 = TextureSampleParameter2D(UVs=@n90) {group=Textures, parameterName=TreadNormal, samplerType=SAMPLERTYPE_Normal, texture=/VehicleCosmetics/Textures/T_BlankNormal}
  NamedRerouteDeclaration(None=@n91.RGBA) {name=NR_TreadNormal, nodeColor=<rgba 0.3828 1 0 1>}
  n92 = NamedRerouteUsage()
  NamedRerouteDeclaration(None=StaticSwitchParameter(True=NamedRerouteUsage(), False=ComponentMask(None=@n92) {r=true}) {desc=L1, group=Colors, parameterName=RimPaintUses2Colors}) {name=ColorizationMasks, nodeColor=<rgba 0 1 1 1>}
  NamedRerouteDeclaration(None=VectorParameter() {group=Colors, parameterName=TreadColor, defaultValue=<rgba 0.3868 0 1 1>}.RGB) {name=NR_TreadColor, nodeColor=<rgba 1 0 0.6797 1>}
  NamedRerouteDeclaration(None=@n22) {name=NR_UseTreadColorInSpec, nodeColor=<rgba 0.1366 1 0 1>}
  n93 = Constant() {r=2}
  n94 = Multiply(
   A=ComponentMask(None=NamedRerouteUsage()) {g=true}
   B=Multiply(
    A=TextureSample(UVs=@n93) {samplerType=SAMPLERTYPE_Normal, texture=/DealershipSharedLibrary/FX/Textures/Noise/T_VCVFX_Matte_N}.RGB
    B=VectorParameter() {group=Tread_Properties, parameterName=TreadDetailNormalIntensity, defaultValue=<rgba 0.375 0.375 0 1>}.RGB
   )
  )
  n95 = Multiply(A=ComponentMask(None=NamedRerouteUsage()) {b=true, g=true, r=true}, B=Constant3Vector() {constant=<rgba 0.625 0.625 1 0>})
  NamedRerouteDeclaration(None=QualitySwitch(Default=Add(A=@n95, B=@n94), Low=@n95)) {name=TreadNormal_Out, nodeColor=<rgba 0 0.25 1 1>}
  n96 = Saturate(None=@n26)
  n97 = NamedRerouteUsage()
  n98 = Constant() {r=0.5}
  n99 = ComponentMask(None=NamedRerouteUsage()) {b=true, g=true, r=true}
  n100 = NamedRerouteUsage()
  NamedRerouteDeclaration(None=MaterialFunctionCall(Base=Multiply(A=@n96, B=@n97), Blend=LinearInterpolate(A=@n98, B=@n99, Alpha=@n100)) {materialFunction=&Blend_Overlay}.Result) {name=TreadDiffuse_Out, nodeColor=<rgba 0.1484 0 1 1>}
  n101 = NamedRerouteUsage()
  n102 = Multiply(A=NamedRerouteUsage(), B=ComponentMask(None=@n101) {r=true})
  n103 = Multiply(A=ComponentMask(None=@n101) {g=true}, B=NamedRerouteUsage())
  NamedRerouteDeclaration(None=Multiply(A=ScalarParameter() {group=Tread_Properties, parameterName=TreadEmissiveBrightness}, B=Add(A=@n102, B=@n103))) {name=TreadEmissive_Out, nodeColor=<rgba 0 0.8125 1 1>}
  NamedRerouteDeclaration(None=TextureSampleParameter2D() {group=Textures, parameterName=RimNormal, samplerType=SAMPLERTYPE_Normal, texture=&T_BlankNormal}.RGB) {name=RimNormal_Out, nodeColor=<rgba 1 0 0.2578 1>}
  n104 = ComponentMask(None=NamedRerouteUsage()) {b=true}
  n105 = LinearInterpolate(
   A=ScalarParameter() {defaultValue=0.5, group=Rim_Properties, parameterName=RimSpecular, sortPriority=1}
   B=ScalarParameter() {defaultValue=0.5, group=Rim_Properties, parameterName=RimSpecular_Painted, sortPriority=1}
   Alpha=NamedRerouteUsage()
  )
  n106 = ComponentMask(None=NamedRerouteUsage()) {b=true}
  n107 = LinearInterpolate(
   A=ScalarParameter() {defaultValue=0.85, group=Tread_Properties, parameterName=TreadSpecular, sortPriority=1}
   B=ScalarParameter() {defaultValue=0.85, group=Tread_Properties, parameterName=TreadSpecular_Painted, sortPriority=1}
   Alpha=NamedRerouteUsage()
  )
  NamedRerouteDeclaration(None=LinearInterpolate(
    A=StaticSwitchParameter(True=@n104, False=@n105) {group=Rim_Properties, parameterName=RimSpecFromTexture}
    B=StaticSwitchParameter(True=@n106, False=@n107) {group=Tread_Properties, parameterName=TreadSpecFromTexture}
    Alpha=NamedRerouteUsage()
   )) {name=Specular_Output, nodeColor=<rgba 0 1 0.1328 1>}
  NamedRerouteDeclaration(None=LinearInterpolate(A=NamedRerouteUsage(), B=NamedRerouteUsage(), Alpha=NamedRerouteUsage())) {name=Emissive_Output, nodeColor=<rgba 0.5938 1 0 1>}
  n108 = ComponentMask(None=NamedRerouteUsage()) {b=true, g=true, r=true}
  n109 = NamedRerouteUsage()
  n110 = Multiply(A=Saturate(None=DotProduct(A=@n109, B=@n109)) {desc=Spec to RimColor}, B=NamedRerouteUsage())
  NamedRerouteDeclaration(None=Multiply(A=Multiply(A=@n108, B=@n110), B=ScalarParameter() {defaultValue=4, parameterName=TreadSpecBrightness})) {name=TreadSpec_Out, nodeColor=<rgba 0.9688 1 0 1>}
  n111 = TextureSampleParameter2D() {group=Textures, parameterName=RimBC, texture=&T_Black}
  NamedRerouteDeclaration(None=@n111.A) {name=NR_RimBC_Alpha, desc=FIX THIS NEXT, nodeColor=<rgba 0.9219 0 1 1>}
  n112 = ComponentMask(None=NamedRerouteUsage()) {g=true}
  n113 = LinearInterpolate(
   A=ScalarParameter() {defaultValue=0.5, group=Rim_Properties, parameterName=RimRoughness, sortPriority=1}
   B=ScalarParameter() {defaultValue=0.5, group=Colors, parameterName=RimRoughness_Painted, sortPriority=1}
   Alpha=NamedRerouteUsage()
  )
  n114 = ComponentMask(None=NamedRerouteUsage()) {g=true}
  n115 = LinearInterpolate(
   A=ScalarParameter() {defaultValue=0.85, group=Tread_Properties, parameterName=TreadRoughness, sortPriority=1}
   B=ScalarParameter() {defaultValue=0.85, group=Colors, parameterName=TreadRoughness_Painted, sortPriority=1}
   Alpha=NamedRerouteUsage()
  )
  NamedRerouteDeclaration(None=LinearInterpolate(
    A=StaticSwitchParameter(True=@n112, False=@n113) {group=Rim_Properties, parameterName=RimRoughnessFromTexture}
    B=StaticSwitchParameter(True=@n114, False=@n115) {group=Tread_Properties, parameterName=TreadRoughnessFromTexture}
    Alpha=NamedRerouteUsage()
   )) {name=Roughness_Output, nodeColor=<rgba 1 0.3047 0 1>}
  NamedRerouteDeclaration(None=StaticSwitchParameter(True=Max(A=@n44, B=@n53), False=Max(A=@n1.R, B=@n1.R)) {group=Rim_Properties, parameterName=Rim_FlooredPaintColor}) {name=NR_RimRGB_ColorizationMasks, nodeColor=<rgba 0 0.7188 1 1>}
  NamedRerouteDeclaration(None=VertexColor().R) {name=NR_VertexColor_R, nodeColor=<rgba 1 0 0.4922 1>}
  n116 = Multiply(A=NamedRerouteUsage(), B=ComponentMask(None=NamedRerouteUsage()) {r=true})
  n117 = ScalarParameter() {defaultValue=3, parameterName=RimEmissiveBrightness}
  n118 = ScalarParameter() {group=Rim_Properties, parameterName=RimAccentEmissiveBrightness}
  n119 = Multiply(A=ComponentMask(None=NamedRerouteUsage()) {g=true}, B=NamedRerouteUsage())
  NamedRerouteDeclaration(None=StaticSwitchParameter(True=Add(A=@n116, B=@n117), False=Multiply(A=@n118, B=@n119)) {group=Rim_Properties, parameterName=RimEmissiveUses2Colors}) {name=NR_RimEmissive_Output, nodeColor=<rgba 0.0312 1 0 1>}
  NamedRerouteDeclaration(None=StaticSwitchParameter(True=Max(A=@n79, B=@n89), False=Max(A=@n4.R, B=@n4.R)) {group=Tread_Properties, parameterName=Tread_FlooredPaintColor}) {name=NR_TreadRGB_ColorizationMasks, nodeColor=<rgba 0.8281 1 0 1>}
  NamedRerouteDeclaration(None=LinearInterpolate(A=NamedRerouteUsage(), B=NamedRerouteUsage(), Alpha=NamedRerouteUsage())) {name=Normal_Output, nodeColor=<rgba 1 0 0.2109 1>}
  NamedRerouteDeclaration(None=@n1.B) {name=RimAOTexture, nodeColor=<rgba 1 0 0.3047 1>}
  NamedRerouteDeclaration(None=@n4.B) {name=TreadAOTexture, nodeColor=<rgba 0.7109 0 1 1>}
  n120 = ScalarParameter() {defaultValue=1, group=Rim_Properties, parameterName=RimAOMultiplier}
  n121 = StaticSwitchParameter(True=NamedRerouteUsage(), False=Constant() {r=1}) {defaultValue=true, group=Rim_Properties, parameterName=RimHasAO}
  NamedRerouteDeclaration(None=Saturate(None=Multiply(A=@n120, B=@n121))) {name=RimAO_Out, nodeColor=<rgba 0.3594 1 0 1>}
  n122 = ScalarParameter() {defaultValue=1, group=Tread_Properties, parameterName=TreadAOMultiplier}
  n123 = StaticSwitchParameter(True=NamedRerouteUsage(), False=Constant() {r=1}) {defaultValue=true, group=Tread_Properties, parameterName=TreadHasAO}
  NamedRerouteDeclaration(None=Saturate(None=Multiply(A=@n122, B=@n123))) {name=TreadAO_Out, nodeColor=<rgba 0.9922 0 1 1>}
  NamedRerouteDeclaration(None=@n4.A) {name=TreadCurvatureTexture, nodeColor=<rgba 1 0 0.1172 1>}
  NamedRerouteDeclaration(None=@n1.A) {name=RimCurvatureTexture, nodeColor=<rgba 0 0.4375 1 1>}
  NamedRerouteDeclaration(None=VectorParameter() {group=Colors, parameterName=CustomColor, defaultValue=<rgba 0.3058 1 0 1>}.RGB) {name=NR_CustomColor, nodeColor=<rgba 0 1 0.9766 1>}
  NamedRerouteDeclaration(None=VectorParameter() {group=Colors, parameterName=AccentColor, defaultValue=<rgba 0 0.6524 1 1>}.RGB) {name=NR_AccentColor, nodeColor=<rgba 1 0.6797 0 1>}
  NamedRerouteDeclaration(None=@n111.RGB) {name=NR_RimBC, nodeColor=<rgba 1 0 0.9844 1>}
  n124 = TextureSampleParameter2D(UVs=@n90) {
   desc="R=Metallic, G=Rough, B=Spec, A=Unused"
   group=Textures
   parameterName=TreadPBR
   samplerType=SAMPLERTYPE_LinearColor
   texture=&T_Black_Linear
  }
  NamedRerouteDeclaration(None=@n124.RGBA) {name=NR_TreadPBR, nodeColor=<rgba 0.2891 1 0 1>}
  NamedRerouteUsage()
  NamedRerouteUsage()
  NamedRerouteUsage()
  PinBase() {reroutePins [
    InputChannel {expression=&self:MaterialExpressionReroute_11}
    FloorValue {expression=&self:MaterialExpressionReroute_12}
   ]}
  PinBase() {reroutePins [
    InputChannel {expression=&self:MaterialExpressionReroute_26}
    FloorValue {expression=&self:MaterialExpressionReroute_28}
   ]}
  PinBase!(Output=@n74) {reroutePins [Output {expression=&self:MaterialExpressionReroute_29}]}
  PinBase!() {reroutePins [
    InputChannel {expression=&self:MaterialExpressionReroute_17}
    FloorValue {expression=&self:MaterialExpressionReroute_18}
   ]}
  PinBase!(Output=@n78) {reroutePins [Output {expression=&self:MaterialExpressionReroute_19}]}
  PinBase!() {reroutePins [
    InputChannel {expression=&self:MaterialExpressionReroute_21}
    FloorValue {expression=&self:MaterialExpressionReroute_22}
   ]}
  PinBase() {reroutePins [
    InputChannel {expression=&self:MaterialExpressionReroute_40}
    FloorValue {expression=&self:MaterialExpressionReroute_41}
   ]}
  PinBase!(Output=@n48) {reroutePins [Output {expression=&self:MaterialExpressionReroute_42}]}
  PinBase!() {reroutePins [
    InputChannel {expression=&self:MaterialExpressionReroute_43}
    FloorValue {expression=&self:MaterialExpressionReroute_44}
   ]}
  PinBase!(Output=@n39) {reroutePins [Output {expression=&self:MaterialExpressionReroute_45}]}
  PinBase!() {reroutePins [
    InputChannel {expression=&self:MaterialExpressionReroute_37}
    FloorValue {expression=&self:MaterialExpressionReroute_39}
   ]}
  PinBase!(Output=@n84) {reroutePins [Output {expression=&self:MaterialExpressionReroute_46}]}
  PinBase(Output=@n52) {reroutePins [Output {expression=&self:MaterialExpressionReroute_20}]}
  PinBase!() {reroutePins [
    InputChannel {expression=&self:MaterialExpressionReroute_24}
    FloorValue {expression=&self:MaterialExpressionReroute_25}
   ]}
  PinBase!(Output=@n43) {reroutePins [Output {expression=&self:MaterialExpressionReroute_23}]}
  PinBase(Output=@n88) {reroutePins [Output {expression=&self:MaterialExpressionReroute_13}]}
 ]}
