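# Material function for vehicle wheel cosmetics that computes a tire pop vertex offset, pulling masked vertices toward the object's pivot.
# The vector from world position to object position is scaled by the TirePop scalar parameter.
# Saturated vertex color green blends from zero to that offset, limiting the deformation to painted vertices.
# The zero side is a vertex normal multiplied by 0, so normals have no effect.
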MaterialFunction /VehicleCosmetics/Wheels/_SharedMaterials/MF_Wheel_TirePop
graph {nodes [
  n1 = VertexNormalWS()
  n2 = ScalarParameter() {parameterName=TirePop}
  n3 = Subtract(A=ObjectPositionWS(), B=WorldPosition().XYZ)
  n4 = VertexColor()
  FunctionOutput(None=LinearInterpolate(A=Multiply(A=@n1) {constB=0}, B=Multiply(A=@n2, B=@n3), Alpha=Saturate(None=@n4.G)))
 ]}
