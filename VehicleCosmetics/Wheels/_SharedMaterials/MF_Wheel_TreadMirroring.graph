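# Material function for vehicle wheels that mirrors tread UVs independently along X and Y.
# Two scalar parameters, TreadUVDirection_X and TreadUVDirection_Y, are saturated and remapped from 0..1 to a -1..1 multiplier, defaulting to unflipped.
# The multipliers scale the texture coordinates for the UV output and are also exposed individually as LerpA and LerpB.
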MaterialFunction /VehicleCosmetics/Wheels/_SharedMaterials/MF_Wheel_TreadMirroring
graph {nodes [
  n1 = LinearInterpolate(Alpha=Saturate(None=ScalarParameter() {defaultValue=1, parameterName=TreadUVDirection_X})) {constA=-1}
  n2 = LinearInterpolate(Alpha=Saturate(None=ScalarParameter() {defaultValue=1, parameterName=TreadUVDirection_Y})) {constA=-1}
  FunctionOutput(None=Multiply(A=TextureCoordinate(), B=AppendVector(A=@n1, B=@n2))) {outputName=UV}
  FunctionOutput(None=@n1) {outputName=LerpA, sortPriority=1}
  FunctionOutput(None=@n2) {outputName=LerpB, sortPriority=2}
 ]}
