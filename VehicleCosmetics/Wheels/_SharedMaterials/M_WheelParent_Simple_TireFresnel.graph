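# Parent material for simple vehicle cosmetic wheels, adding an optional fresnel-style color highlight to tires.
# MF_WheelParent_Simple output enters a Tire_Exists static switch (off by default), then passes through damage and de-rez transition functions.
# The highlight powers an uninverted view-dot-normal, peaking on camera-facing areas, to push base color toward brightened Tire_ColorA.
# The tread-mirrored normal shapes it, AO optionally attenuates it, and a vertex-color mask confines it.
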Material /VehicleCosmetics/Wheels/_SharedMaterials/M_WheelParent_Simple_TireFresnel
graph {
 properties {
  bAutomaticallySetUsageInEditor=false
  bUseMaterialAttributes=true
  bUsedWithMeshDeformer=true
  bUsedWithSkeletalMesh=true
  pixelDepthOffsetMode=PDOM_Legacy
  usedShadingModels=Default Lit
 }
 nodes [
  Composite() {subgraphName=Normal w/Tread Mirroring}
  Composite() {subgraphName=Color Settings A-C}
  Composite() {subgraphName=Default Settings}
  n1 = MaterialFunctionCall() {materialFunction=&MF_WheelParent_Simple}
  Composite!("In (MA)"=@n1.MaterialAttributes, TireAO=@n1.MaterialAttributes, TireVCMask=@n1.MaterialAttributes) {subgraphName=Tire Fresnel}
  MaterialExpressionReroute_0 = Reroute(None=@n1.MaterialAttributes)
  n2 = GetMaterialAttributes(None=@MaterialExpressionReroute_0."In (MA)") {attributeGetTypes [#1]}
  n3 = CameraVectorWS()
  n4 = Power(
   Base=DotProduct(
    A=@MaterialExpressionReroute_0.Tire_Normal
    B=Transform(None=@n3) {transformSourceType=TRANSFORMSOURCE_World, transformType=TRANSFORM_Tangent}
   )
   Exp=ScalarParameter() {defaultValue=6, group=5_Special Effects, parameterName=Tire_FresnelExponent, sortPriority=50}
  )
  n5 = Saturate(None=Multiply(
    A=ConstantBiasScale(None=@n4) {bias=0.0156}
    B=ScalarParameter() {defaultValue=1, group=5_Special Effects, parameterName=Tire_FresnelPower, sortPriority=50}
   ))
  MaterialExpressionReroute_1 = Reroute(None=@n1.TireRGBAMask)
  n6 = ComponentMask(None=@MaterialExpressionReroute_1.TireAO) {a=true, desc=Tire AO}
  n7 = Saturate(None=ComponentMask(None=@MaterialExpressionReroute_1.RoughMetalAOTint) {a=true})
  n8 = LinearInterpolate(
   A=@n2.Base Color
   B=Multiply(
    A=ScalarParameter() {defaultValue=16, group=5_Special Effects, parameterName=Tire_FresnelBrightness, sortPriority=50}
    B=@MaterialExpressionReroute_0.ColorA
   )
   Alpha=LinearInterpolate(A=@n5, B=Multiply(A=@n6, B=@n5), Alpha=NamedRerouteDeclaration(None=@n7) {name=Bool_AO_Diffuse, nodeColor=<rgba 1 0.5859 0 1>})
  )
  MaterialExpressionReroute_2 = Reroute(None=@n1.VertexColorR)
  n9 = StaticSwitchParameter(
   True=BlendMaterialAttributes(
    A=@MaterialExpressionReroute_0."In (MA)"
    B=SetMaterialAttributes!(MaterialAttributes=@n2.MaterialAttributes, Base Color=@n8) {attributeSetTypes [#1]}
    Alpha=@MaterialExpressionReroute_2.TireVCMask
   )
   False=@MaterialExpressionReroute_0."In (MA)"
  ) {desc="Tire == True, No Tire == False. False is default.", group=5_Special Effects, parameterName=Tire_Exists}
  MaterialExpressionReroute_3 = Reroute(None=@n9)
  MaterialFunctionCall(In=MaterialFunctionCall(Material=@MaterialExpressionReroute_3.Output) {materialFunction=&MF_BaseVehicle_Damage}.Result) {materialFunction=&MF_VehicleDeRez_Transition}
  n10 = MaterialFunctionCall() {
   desc="UV - For all Tread Mirrors. Lerp A && B - For Normals"
   materialFunction=&/MF_Wheel_TreadMirroring
  }
  NamedRerouteDeclaration(None=@n10.LerpA) {name=Tread_LerpA, nodeColor=<rgba 0 1 0.6953 1>}
  NamedRerouteDeclaration(None=@n10.LerpB) {name=Tread_LerpB, nodeColor=<rgba 0 1 0.6953 1>}
  NamedRerouteDeclaration(None=@n10.UV) {name=Tread_UV, nodeColor=<rgba 0 1 0.6953 1>}
  PinBase()
  n11 = VectorParameter() {
   desc=Color fill for R value of RGB Mask
   group=3_Tire Properties
   parameterName=Tire_ColorA
   channelNames {a=Unused, b="", g="", r=""}
   defaultValue=<rgba 1 0.1248 0 1>
  }
  n12 = VectorParameter() {
   desc="Emissive, Roughness, Metallic values for ColorA. Emissive has no limit, Roughness && Metallic are clamped between 0 && 1."
   group=3_Tire Properties
   parameterName=Tire_ColorA_PBR
   channelNames {a=Unused, b=Metallic, g=Roughness, r=Emissive}
   defaultValue=<rgba 0 0.7 0 1>
  }
  n13 = VectorParameter() {
   desc=Color fill for G value of RGB Mask
   group=3_Tire Properties
   parameterName=Tire_ColorB
   channelNames {a=Unused, b="", g="", r=""}
   defaultValue=<rgba 0 1 0.5029 1>
  }
  n14 = VectorParameter() {
   desc="Emissive, Roughness, Metallic values for ColorB. Emissive has no limit, Roughness && Metallic are clamped between 0 && 1."
   group=3_Tire Properties
   parameterName=Tire_ColorB_PBR
   channelNames {a=Unused, b=Metallic, g=Roughness, r=Emissive}
   defaultValue=<rgba 0 0.7 0 1>
  }
  n15 = VectorParameter() {
   desc=Color fill for B value of RGB Mask
   group=3_Tire Properties
   parameterName=Tire_ColorC
   channelNames {a=Unused, b="", g="", r=""}
   defaultValue=<rgba 0.5029 0 1 1>
  }
  n16 = VectorParameter() {
   desc="Emissive, Roughness, Metallic values for ColorC. Emissive has no limit, Roughness && Metallic are clamped between 0 && 1."
   group=3_Tire Properties
   parameterName=Tire_ColorC_PBR
   channelNames {a=Unused, b=Metallic, g=Roughness, r=Emissive}
   defaultValue=<rgba 0 0.7 0 1>
  }
  PinBase!(ColorA=@n11, ColorA_PBR=@n12, ColorB=@n13, ColorB_PBR=@n14, ColorC=@n15, ColorC_PBR=@n16) {reroutePins [
    ColorA {expression=&self:MaterialGraph_0.MaterialGraphNode_Composite_0.Tire Fresnel.MaterialGraphNode_Composite_0.Color Settings A-C.MaterialExpressionReroute_0}
    ColorA_PBR {expression=&self:MaterialGraph_0.MaterialGraphNode_Composite_0.Tire Fresnel.MaterialGraphNode_Composite_0.Color Settings A-C.MaterialExpressionReroute_1}
    ColorB {expression=&self:MaterialGraph_0.MaterialGraphNode_Composite_0.Tire Fresnel.MaterialGraphNode_Composite_0.Color Settings A-C.MaterialExpressionReroute_2}
    ColorB_PBR {expression=&self:MaterialGraph_0.MaterialGraphNode_Composite_0.Tire Fresnel.MaterialGraphNode_Composite_0.Color Settings A-C.MaterialExpressionReroute_3}
    ColorC {expression=&self:MaterialGraph_0.MaterialGraphNode_Composite_0.Tire Fresnel.MaterialGraphNode_Composite_0.Color Settings A-C.MaterialExpressionReroute_4}
    ColorC_PBR {expression=&self:MaterialGraph_0.MaterialGraphNode_Composite_0.Tire Fresnel.MaterialGraphNode_Composite_0.Color Settings A-C.MaterialExpressionReroute_5}
   ]}
  PinBase!() {reroutePins [
    "In (MA)" {expression=&self:MaterialGraph_0.MaterialGraphNode_Composite_0.Tire Fresnel.MaterialExpressionReroute_0}
    TireAO {expression=&self:MaterialGraph_0.MaterialGraphNode_Composite_0.Tire Fresnel.MaterialExpressionReroute_1}
    TireVCMask {expression=&self:MaterialGraph_0.MaterialGraphNode_Composite_0.Tire Fresnel.MaterialExpressionReroute_2}
   ]}
  PinBase!(Output=@n9) {reroutePins [Output {expression=&self:MaterialGraph_0.MaterialGraphNode_Composite_0.Tire Fresnel.MaterialExpressionReroute_3}]}
  PinBase!()
  n17 = TextureSampleParameter2D(UVs=NamedRerouteUsage()) {group=1_Input Textures, parameterName=Tire_Normal, samplerType=SAMPLERTYPE_Normal, texture=/VehicleCosmetics/Textures/T_BlankNormal_B}
  n18 = NamedRerouteUsage()
  n19 = NamedRerouteUsage()
  PinBase!(Tire_Normal=MaterialFunctionCall(R=Multiply(A=@n17.R, B=@n18) {constB=-1}, G=Multiply(A=@n17.G, B=@n19) {constB=-1}, B=@n17.B) {materialFunction=&AppendMany}.RGB) {reroutePins [Tire_Normal {expression=&self:MaterialGraph_0.MaterialGraphNode_Composite_0.Tire Fresnel.MaterialGraphNode_Composite_1.Normal w/Tread Mirroring.MaterialExpressionReroute_0}]}
  PinBase!()
  n20 = ScalarParameter() {
   defaultValue=0.7
   desc=General Roughness that takes effect in absence of masking.
   group=4_Surface Defaults
   parameterName=Tire_RoughnessBase
   sliderMax=1
  }
  n21 = ScalarParameter() {desc=General Metalness that takes effect in absence of masking., group=4_Surface Defaults, parameterName=Tire_MetallicBase, sliderMax=1}
  n22 = ScalarParameter() {
   defaultValue=1
   desc="0==False, 1==True: AO gradient affects Emissive values. Use this for Emissives with depth. Values between 0 && 1 are allowed."
   group=5_Special Effects
   parameterName=Tire_AOTintsEmissive
  }
  n23 = ScalarParameter() {
   desc="0==False, 1==True: Blends AO with Diffuse. Values between 0 && 1 are allowed."
   group=5_Special Effects
   parameterName=Tire_AOTintsDiffuse
  }
  n24 = MaterialFunctionCall(R=@n20, G=@n21, B=@n22, A=@n23) {desc="Vector4 container for Rough, Metallic, AO settings", materialFunction=&AppendMany}
  PinBase!(RoughMetalAOTint=@n24.RGBA) {reroutePins [RoughMetalAOTint {expression=&self:MaterialGraph_0.MaterialGraphNode_Composite_0.Tire Fresnel.MaterialGraphNode_Composite_2.Default Settings.MaterialExpressionReroute_1}]}
  MaterialExpressionReroute_4 = Reroute(None=@n15)
  MaterialExpressionReroute_5 = Reroute(None=@n16)
 ]
}
