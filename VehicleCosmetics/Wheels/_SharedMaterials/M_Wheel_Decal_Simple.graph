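# Simple masked, lit master material for wheel decals in vehicle cosmetics, usable on skeletal meshes and mesh deformers.
# A static switch optionally mirrors U for the shared decal, mask and normal UVs.
# Decal color is multiplied by a tint and a selectable mask channel, alongside scalar metallic and roughness.
# A second switch sources scaled emissive from either decal color or another selectable mask channel.
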Material /VehicleCosmetics/Wheels/_SharedMaterials/M_Wheel_Decal_Simple
graph {
 properties {
  bAutomaticallySetUsageInEditor=false
  bCastRayTracedShadows=false
  bUsedWithMeshDeformer=true
  bUsedWithSkeletalMesh=true
  blendMode=BLEND_Masked
  usedShadingModels=Default Lit
 }
 nodes [
  n1 = TextureCoordinate()
  n2 = ComponentMask(None=@n1) {r=true}
  n3 = StaticSwitchParameter(True=AppendVector(A=OneMinus(None=@n2), B=ComponentMask(None=@n1) {g=true}), False=@n1) {group=Base, parameterName=Mirror Decal, sortPriority=-10}
  n4 = TextureSampleParameter2D(UVs=@n3) {group=Base, parameterName=Decal, sortPriority=-5, texture=/VehicleCosmetics/Textures/T_White}
  n5 = TextureSampleParameter2D(UVs=@n3) {group=Base, parameterName=Mask, sortPriority=-3, texture=&T_White}
  Multiply(
   A=Multiply(A=@n4.RGB, B=VectorParameter() {group=Base, parameterName=Mask Tint, sortPriority=0, defaultValue=<rgba 1 1 1 1>}.RGB)
   B=ChannelMaskParameter(None=@n5.RGBA) {group=Base, parameterName=MaskChannelPicker, sortPriority=0}.R
  )
  n6 = TextureSampleParameter2D(UVs=@n3) {group=Base, parameterName=Decal Normal, samplerType=SAMPLERTYPE_Normal, sortPriority=-4, texture=/VehicleCosmetics/Textures/T_BlankNormal_B}
  Reroute(None=@n6.RGB)
  ScalarParameter() {group=Base, parameterName=Metallic, sortPriority=3}
  ScalarParameter() {defaultValue=1, group=Base, parameterName=Roughness, sortPriority=4}
  n7 = ScalarParameter() {group=Base, parameterName=Emissive, sortPriority=5}
  StaticSwitchParameter(
   True=Multiply(A=@n4.RGB, B=@n7)
   False=Multiply(A=ChannelMaskParameter(None=@n5.RGBA) {group=Base, parameterName=EmissiveChannelPicker, sortPriority=1}.R, B=@n7)
  ) {group=Base, parameterName=EmmisiveFromBC, sortPriority=0}
 ]
}
