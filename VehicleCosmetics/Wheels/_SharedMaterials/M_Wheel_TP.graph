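# Translucent, lit master material for vehicle wheel cosmetics, usable on skeletal meshes and mesh deformers.
# Tinted albedo plus tint is summed with a vertex-normal cubemap reflection and a Fresnel-driven CustomColor glow, then multiplied by that glow again.
# Separate chains scale the normal map and bias-clamp the reflection.
# A static switch selects albedo alpha or scalar opacity; material output connections are not shown.
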Material /VehicleCosmetics/Wheels/_SharedMaterials/M_Wheel_TP
graph {
 properties {
  bAutomaticallySetUsageInEditor=false
  bUsedWithMeshDeformer=true
  bUsedWithSkeletalMesh=true
  blendMode=BLEND_Translucent
  translucencyLightingMode=TLM_VolumetricDirectional
  usedShadingModels=Default Lit
 }
 nodes [
  n1 = TextureSampleParameterCube!(UVs=ReflectionVectorWS(CustomWorldNormal=VertexNormalWS()), Apply View MipBias) {group=Base Textures, parameterName=CubeMap, sortPriority=2, texture=/Game/Packages/Fortress_Sky/TexturesHDR/FrontendCubeMap_output}
  n2 = Multiply(A=@n1.RGB, B=ScalarParameter() {defaultValue=1, group=Base Parameters, parameterName=CubeMap Intensity, sortPriority=4})
  Clamp(None=ConstantBiasScale(None=@n2))
  n3 = TextureSampleParameter2D() {
   group=Base Textures
   parameterName=Normal
   samplerType=SAMPLERTYPE_Normal
   sortPriority=1
   texture=&FlatNormalBlue
  }
  Multiply(A=@n3.RGB, B=Constant3Vector() {constant=<rgba 0.625 0.625 1 0>})
  n4 = TextureSampleParameter2D() {group=Base Textures, parameterName=Albedo, sortPriority=0}
  n5 = VectorParameter() {group=Base Parameters, parameterName=Albedo Tint, sortPriority=0, defaultValue=<rgba 0.33 0.33 0.33 0>}
  n6 = Add(A=Multiply(A=@n4.RGB, B=@n5.RGB), B=@n5.RGB)
  n7 = Fresnel()
  n8 = Multiply(
   A=@n7
   B=Multiply(
    A=VectorParameter() {group=Custom Color Properties, parameterName=CustomColor, defaultValue=<rgba 0.59 0 0 1>}.RGB
    B=ScalarParameter() {defaultValue=1, group=Base Parameters, parameterName=Emissive Amount, sortPriority=5}
   )
  )
  Multiply(A=Add(A=Add(A=@n6, B=@n2), B=LinearInterpolate(A=@n7, B=@n8) {constAlpha=1}), B=@n8)
  StaticSwitchParameter(True=@n4.A, False=ScalarParameter() {defaultValue=1.5, group=Base Parameters, parameterName=Opacity, sortPriority=3}) {group=Base Parameters, parameterName=Use Albedo Opacity, sortPriority=1}
 ]
}
