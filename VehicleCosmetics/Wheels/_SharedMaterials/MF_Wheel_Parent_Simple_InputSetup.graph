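# Material function that builds shared surface attributes for simple vehicle cosmetic wheel materials.
# An RGBA mask layers three color regions, each with emissive, roughness and metallic values, over base settings; its alpha carries AO.
# A rounded 0-1 switch controls optional diffuse texture blending, and clamped base-setting flags control AO on diffuse and emissive.
# Outputs material attributes including the optional normal, plus packed masks as a second output.
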MaterialFunction /VehicleCosmetics/Wheels/_SharedMaterials/MF_Wheel_Parent_Simple_InputSetup
graph {nodes [
  n1 = FunctionInput() {
   description="0==False, 1==True: Blends Input Diffuse Texture with RGB Colors. Leave at 0 if there's no Diffuse."
   inputName=DiffuseTextureEnable
   inputType=FunctionInput_Scalar
   sortPriority=4
  }
  Composite!(None=@n1) {subgraphName=Clamp 0-1}
  n2 = SetMaterialAttributes!(
   MaterialAttributes
   BaseColor=NamedRerouteUsage()
   Metallic=NamedRerouteUsage()
   Roughness=NamedRerouteUsage()
   EmissiveColor=NamedRerouteUsage()
   Normal=NamedRerouteUsage()
  ) {attributeSetTypes [
    #1
    #2
    #3
    #4
    #5
   ]}
  FunctionOutput(None=@n2)
  FunctionOutput(None=MaterialFunctionCall(R=NamedRerouteUsage(), G=NamedRerouteUsage(), B=NamedRerouteUsage(), A=NamedRerouteUsage()) {materialFunction=&AppendMany}.RGBA) {outputName=Masks Out, sortPriority=1}
  NamedRerouteDeclaration(None=FunctionInput() {description=Vector Color, inputName=ColorA_Color, sortPriority=12}) {name=ColorA_RGB_Out, nodeColor=<rgba 0 0.1094 1 1>}
  NamedRerouteDeclaration(None=FunctionInput() {description=Vector Color, inputName=ColorC_Color, sortPriority=20}) {name=ColorC_RGB_Out, nodeColor=<rgba 0 0 1 1>}
  n3 = MaterialFunctionCall!(RGBA=FunctionInput() {inputName=MaskRGBA, inputType=FunctionInput_Vector4, sortPriority=8}) {materialFunction=/VehicleCosmetics/SharedMaterials/MaterialFunctions/MF_RGBA_Splitter}
  NamedRerouteDeclaration(None=@n3.A) {name=AlphaAO_Mask_Out, nodeColor=<rgba 1 1 1 1>}
  n4 = FunctionInput() {description=Normal Texture - Optional, inputName=Normal_Texture, sortPriority=2, previewValue=<1 1 1 1>}
  NamedRerouteDeclaration(None=@n4) {name=Normal_Tex_Out, nodeColor=<rgba 0.4606 0 1 1>}
  n5 = MaterialFunctionCall(Color=FunctionInput() {inputName=ColorC_PBR_Values, sortPriority=21}) {materialFunction=&SplitComponents}
  NamedRerouteDeclaration(None=Saturate(None=@n5.G)) {name=ColorC_Rough_Out, nodeColor=<rgba 0 0 1 1>}
  n6 = MaterialFunctionCall(Color=FunctionInput() {inputName=ColorB_PBR_Values, sortPriority=17}) {materialFunction=&SplitComponents}
  NamedRerouteDeclaration(None=Saturate(None=@n6.G)) {name=ColorB_Rough_Out, nodeColor=<rgba 0 1 0 1>}
  n7 = MaterialFunctionCall(Color=FunctionInput() {inputName=ColorA_PBR_Values, sortPriority=13}) {materialFunction=&SplitComponents}
  NamedRerouteDeclaration(None=Saturate(None=@n7.G)) {name=ColorA_Rough_Out, nodeColor=<rgba 0.0312 1 0 1>}
  n8 = NamedRerouteUsage()
  n9 = Multiply(A=NamedRerouteUsage(), B=NamedRerouteUsage())
  n10 = NamedRerouteUsage()
  n11 = NamedRerouteUsage()
  n12 = NamedRerouteUsage()
  n13 = LinearInterpolate(
   A=LinearInterpolate(A=LinearInterpolate(B=@n9, Alpha=@n10) {constB=0}, B=Multiply(A=@n11, B=@n12), Alpha=NamedRerouteUsage())
   B=Multiply(A=NamedRerouteUsage(), B=NamedRerouteUsage())
   Alpha=NamedRerouteUsage()
  )
  n14 = NamedRerouteUsage()
  NamedRerouteDeclaration(None=LinearInterpolate(A=Multiply(A=@n8, B=@n13), B=Multiply(A=@n14, B=@n13) {constB=0}, Alpha=NamedRerouteUsage())) {name=Emissive_Rim_Out, nodeColor=<rgba 1 0.8985 0 1>}
  n15 = LinearInterpolate(A=LinearInterpolate(B=NamedRerouteUsage(), Alpha=NamedRerouteUsage()) {constB=0}, B=NamedRerouteUsage(), Alpha=NamedRerouteUsage())
  n16 = NamedRerouteUsage()
  n17 = NamedRerouteUsage()
  NamedRerouteDeclaration(None=LinearInterpolate(A=NamedRerouteUsage(), B=LinearInterpolate(A=@n15, B=@n16, Alpha=@n17), Alpha=NamedRerouteUsage())) {name=Metallic_Rim_Out, nodeColor=<rgba 0.542 1 0.8687 1>}
  NamedRerouteDeclaration(None=FunctionInput() {description=Vector Color, inputName=ColorB_Color, sortPriority=16}) {name=ColorB_RGB_Out, nodeColor=<rgba 0 1 0 1>}
  n18 = NamedRerouteUsage()
  n19 = LinearInterpolate(B=NamedRerouteUsage(), Alpha=@n18)
  n20 = NamedRerouteUsage()
  n21 = NamedRerouteUsage()
  n22 = NamedRerouteUsage()
  n23 = Max(A=@n18, B=@n21)
  n24 = LinearInterpolate(
   A=LinearInterpolate(A=NamedRerouteUsage(), B=NamedRerouteUsage(), Alpha=NamedRerouteUsage())
   B=LinearInterpolate(A=LinearInterpolate(A=@n19, B=@n20, Alpha=@n21), B=NamedRerouteUsage(), Alpha=@n22)
   Alpha=NamedRerouteDeclaration(None=Max(A=@n23, B=@n22)) {name=Mask_Rim_All, nodeColor=<rgba 0 0.9511 1 1>}
  )
  n25 = NamedRerouteUsage()
  NamedRerouteDeclaration(None=LinearInterpolate(A=@n24, B=Multiply(A=@n25, B=@n24), Alpha=NamedRerouteUsage()) {constAlpha=0}) {name=Diffuse_Rim_Blended_Out, nodeColor=<rgba 0.7552 0.1522 0 1>}
  NamedRerouteDeclaration(None=Saturate(None=@n6.B)) {name=ColorB_Metallic_Out, nodeColor=<rgba 0 1 0 1>}
  NamedRerouteDeclaration(None=Saturate(None=@n5.B)) {name=ColorC_Metallic_Out, nodeColor=<rgba 0 0 1 1>}
  NamedRerouteDeclaration(None=Saturate(None=@n7.B)) {name=ColorA_Metallic_Out, nodeColor=<rgba 0.5937 1 0 1>}
  NamedRerouteDeclaration(None=@n7.R) {name=ColorA_Emis_Out, nodeColor=<rgba 1 0 0.9141 1>}
  NamedRerouteDeclaration(None=@n6.R) {name=ColorB_Emis_Out, nodeColor=<rgba 0 1 0 1>}
  n26 = FunctionInput() {description=Diffuse Texture - Optional, inputName=Diffuse_Texture, sortPriority=3, previewValue=<1 1 1 1>}
  NamedRerouteDeclaration(None=@n26) {name=Diffuse_Tex_Out, nodeColor=<rgba 1 0 0.0938 1>}
  NamedRerouteDeclaration(None=@n5.R) {name=ColorC_Emis_Out, nodeColor=<rgba 0 0 1 1>}
  n27 = MaterialFunctionCall(RGBA=FunctionInput() {inputName=Base_Settings, inputType=FunctionInput_Vector4}) {materialFunction=&MF_RGBA_Splitter}
  NamedRerouteDeclaration(None=@n27.R) {name=Rough_Value, nodeColor=<rgba 1 0.491 0 1>}
  NamedRerouteDeclaration(None=@n27.G) {name=Metal_Value, nodeColor=<rgba 1 0.4922 0 1>}
  NamedRerouteDeclaration(None=Saturate(None=@n27.A)) {name=Bool_AO_Diffuse, nodeColor=<rgba 1 0.5859 0 1>}
  MaterialExpressionReroute_2 = Reroute(None=@n1)
  n28 = Saturate(None=Round(None=Abs(None=@MaterialExpressionReroute_2.Input)))
  MaterialExpressionReroute_1 = Reroute(None=@n28)
  NamedRerouteDeclaration(None=@MaterialExpressionReroute_1.Output) {name=Bool_Diffuse_Color, nodeColor=<rgba 1 0.5859 0 1>}
  NamedRerouteDeclaration(None=Saturate(None=@n27.B)) {name=Bool_AO_Emis, nodeColor=<rgba 1 0.5859 0 1>}
  NamedRerouteDeclaration(None=@n3.R) {name=ColorA_Mask_Out, nodeColor=<rgba 0 1 0.2031 1>}
  n29 = FunctionInput() {
   description="Albedo/Diffuse color control in absence of RGB maps, or to fill in the gaps. Leave white if using Diffuse Texture instead."
   inputName=ColorBase
   sortPriority=7
   previewValue=<1 1 1 1>
  }
  NamedRerouteDeclaration(None=@n29) {name=Diffuse_Color_Value, nodeColor=<rgba 1 0.491 0 1>}
  n30 = LinearInterpolate(A=LinearInterpolate(B=NamedRerouteUsage(), Alpha=NamedRerouteUsage()) {constA=1}, B=NamedRerouteUsage(), Alpha=NamedRerouteUsage())
  n31 = NamedRerouteUsage()
  n32 = NamedRerouteUsage()
  NamedRerouteDeclaration(None=LinearInterpolate(A=NamedRerouteUsage(), B=LinearInterpolate(A=@n30, B=@n31, Alpha=@n32), Alpha=NamedRerouteUsage())) {name=Roughness_Rim_Out, nodeColor=<rgba 0 1 0.0156 1>}
  NamedRerouteDeclaration(None=@n3.G) {name=ColorB_Mask_Out, nodeColor=<rgba 0 1 0 1>}
  NamedRerouteDeclaration(None=@n3.B) {name=ColorC_Mask_Out, nodeColor=<rgba 0 0 1 1>}
  PinBase() {reroutePins [Input {expression=&self:MaterialExpressionReroute_2}]}
  PinBase!(Output=@n28) {reroutePins [Output {expression=&self:MaterialExpressionReroute_1}]}
 ]}
